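# Material function for the Cube Ninja cape that layers glowing runes and edge glow over incoming material attributes.
# A noise texture distorts the tiling of a panning throb mask, which pulses rune emissive between minimum and maximum brightness.
# Two emissive masks place the colored runes and edges; the rune term is added twice.
# An override scalar darkens base color and shifts metallic, roughness and specular, mostly limited by rune or sphere masks.
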MaterialFunction /BRCosmetics/Accessories/FORT_Capes/Cape_M_MED_Cube_Ninja/Materials/MF_CubeNinja
graph {nodes [
  n1 = GetMaterialAttributes(None=FunctionInput() {inputType=FunctionInput_MaterialAttributes}) {attributeGetTypes [
    #1
    #2
    #3
    #4
   ]}
  n2 = ScalarParameter() {defaultValue=0.1, group=CubeNinja, parameterName=CN_RuneBrightness-Min}
  n3 = ScalarParameter() {defaultValue=2, group=CubeNinja, parameterName=CN_RuneBrightness-Max}
  n4 = AppendVector(A=ScalarParameter() {group=PanningEmissive, parameterName=UVOffsetU}, B=ScalarParameter() {group=PanningEmissive, parameterName=UVOffsetV})
  n5 = ScalarParameter() {group=PanningEmissive, parameterName=UVOffsetU}
  n6 = ScalarParameter() {group=PanningEmissive, parameterName=UVOffsetV}
  n7 = TextureSampleParameter2D(UVs=Add(A=AppendVector(A=@n5, B=@n6), B=TextureCoordinate())) {
   group=PanningEmissive
   parameterName=UVScaleMap
   samplerSource=SSM_Wrap_WorldGroupSettings
   samplerType=SAMPLERTYPE_LinearColor
   texture=/Engine/Functions/Engine_MaterialFunctions02/ExampleContent/Textures/OffsetNoiseDistanceFields_ForNormals
  }
  n8 = ScalarParameter() {defaultValue=0.4, group=PanningEmissive, parameterName=UVScaleMapIntensity}
  n9 = Saturate(None=Add(A=@n7.RGB, B=OneMinus(None=@n8)))
  n10 = ScalarParameter() {defaultValue=0.5, group=PanningEmissive, parameterName=Emissive Tiling}
  n11 = Multiply(A=TextureCoordinate(), B=ComponentMask(None=Multiply(A=@n9, B=@n10)) {g=true, r=true})
  n12 = ScalarParameter() {group=PanningEmissive, parameterName=GradientSpeed_X}
  n13 = ScalarParameter() {defaultValue=0.2, group=PanningEmissive, parameterName=GradientSpeed_Y}
  n14 = TextureSample(UVs=Panner(Coordinate=Add(A=@n4, B=@n11), Speed=AppendVector(A=@n12, B=@n13))) {texture=/Game/Effects/Fort_Effects/Textures/Taker/T_ThrobMask}
  n15 = ComponentMask(None=@n14.RGB) {r=true}
  n16 = TextureSample() {texture=/BRCosmetics/Accessories/FORT_Capes/Cape_M_MED_Cube_Ninja/FX/T_Cube_Ninja_Male_Emissive_Mask_01}
  n17 = VectorParameter() {parameterName=CN_RuneEmissiveColor, defaultValue=<rgba 0.0347 0 0.4479 1>}
  n18 = Multiply(A=Multiply(A=LinearInterpolate(A=@n2, B=@n3, Alpha=@n15), B=@n16.RGB), B=@n17.RGB)
  n19 = Multiply(
   A=Multiply(
    A=TextureSample() {texture=/BRCosmetics/Accessories/FORT_Capes/Cape_M_MED_Cube_Ninja/FX/T_Cube_Ninja_Male_Emissive_Mask_02}.RGB
    B=ScalarParameter() {defaultValue=0.5, group=CubeNinja, parameterName=CN_EdgeEmsMult}
   )
   B=VectorParameter() {parameterName=CN_EgdeEmissiveColor, defaultValue=<rgba 0 0.3622 1 1>}.RGB
  )
  n20 = ScalarParameter() {group=CubeNinja, parameterName=CN_BaseColorOverideAmount}
  n21 = Add(A=OneMinus(None=@n20)) {constB=0.7}
  n22 = LinearInterpolate(
   A=@n1.BaseColor
   B=Multiply(A=@n1.BaseColor, B=Saturate(None=@n21))
   Alpha=SphereMask(A=TextureCoordinate(), B=Constant2Vector() {g=0.72, r=0.72}, Radius=Constant() {r=1}, Hardness=Constant())
  )
  n23 = Constant() {r=0.03}
  n24 = OneMinus(None=ScalarParameter() {group=CubeNinja, parameterName=CN_BaseColorOverideAmount})
  n25 = Saturate(None=Add(A=@n16.R, B=Multiply(A=@n24) {constB=-1}))
  n26 = SetMaterialAttributes!(
   MaterialAttributes=@n1.MaterialAttributes
   EmissiveColor=Add(A=Add(A=@n18, B=@n19), B=@n18)
   BaseColor=LinearInterpolate(
    A=LinearInterpolate(A=@n22, B=@n23, Alpha=@n25)
    B=ScalarParameter() {defaultValue=0.03, parameterName=CN_RuneBaseColor}
    Alpha=Multiply(A=@n16.R, B=@n14.RGB)
   ) {desc=BaseColor}
   Metallic=LinearInterpolate(A=@n1.Metallic, B=Constant() {r=0.2}, Alpha=@n25) {desc=Metallic}
   Roughness=LinearInterpolate(A=@n1.Roughness, B=Constant() {r=0.4}, Alpha=@n25) {desc=Roughness}
   Specular=LinearInterpolate(A=@n1.Specular, B=Constant() {r=0.001}, Alpha=@n20) {desc=Specular}
  ) {attributeSetTypes [
    #5
    #1
    #2
    #3
    #4
   ]}
  FunctionOutput(None=@n26)
  Multiply(A=Multiply(A=@n16.RGB, B=ScalarParameter() {defaultValue=1, group=CubeNinja, parameterName=CN_RuneEmsMult}), B=@n17.RGB)
 ]}
