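# Master material for the Cube Ninja cape backpack cosmetic, Default Lit and usable on skeletal meshes and mesh deformers.
# It works in material attributes, chaining shared character PreFX, the cosmetic-specific MF_CubeNinja function, shared PostFX, then a CINE character post-FX function.
# PreFX's AO output also feeds PostFX's AO, Distance Blend, Fresnel, Gmap and Pre Skinned Local Position inputs.
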Material /BRCosmetics/Accessories/FORT_Capes/Cape_M_MED_Cube_Ninja/Materials/M_MED_CubeNinja_Backpack_Master
graph {
 properties {
  bUseMaterialAttributes=true
  bUsedWithMeshDeformer=true
  bUsedWithSkeletalMesh=true
  pixelDepthOffsetMode=PDOM_Legacy
  usedShadingModels=Default Lit
 }
 nodes [
  n1 = MaterialFunctionCall() {materialFunction=&BaseFN_CharacterFunction_PreFX}
  n2 = MaterialFunctionCall(
   AO=@n1.AO
   Distance Blend=@n1.AO
   Fresnel=@n1.AO
   Gmap=@n1.AO
   Pre Skinned Local Position=@n1.AO
   Material=MaterialFunctionCall!(In=@n1.Result) {materialFunction=&/MF_CubeNinja}.Result
  ) {materialFunction=&BaseFN_CharacterFunction_PostFX}
  MaterialFunctionCall(In=@n2.Result) {materialFunction=&CINE_MF_CharPostFX}
 ]
}
