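# Physics asset for the female Cube Ninja cape backpack cosmetic, defining capsule bodies for the cape bones and torso.
# The cape chains from backpack_root through dyn_cape05, then dyn_cape01 to 04, with narrow twist and swing1 limits widening toward the tip.
# Pelvis and five spine capsules form a separate torso chain with free angular motion, with no constraint to the cape.
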PhysicsAsset /BRCosmetics/Accessories/FORT_Capes/Cape_M_MED_Cube_Ninja/Meshes/M_MED_Cube_Ninja_Pack_Physics_Female
bodies {
 backpack_root [{shapeType=Capsule, length=18.5264, radius=3.2354, center=<0.0003 -4.7319 8.6288>, rotation=<rot 0 90.4255 -89.0773>}]
 dyn_cape01 [{shapeType=Capsule, length=16.0127, radius=3.5986, center=<4.5022 -2 0>}]
 dyn_cape02 [{shapeType=Capsule, length=17.1009, radius=2.9453, center=<4.2517 -2 0>}]
 dyn_cape03 [{shapeType=Capsule, length=15.3653, radius=2.0779, center=<2.9059 -1.7 0>}]
 dyn_cape04 [{shapeType=Capsule, length=17.4989, radius=1.0274, center=<0 -1.5 0>}]
 dyn_cape05 [{shapeType=Capsule, length=0.505, radius=0.505}]
 pelvis [{shapeType=Capsule, length=8.4637, radius=13.2565, rotation=<rot 0 -90 0>}]
 spine_01 [{shapeType=Capsule, length=4.8113, radius=11.9291, rotation=<rot 0 -90 0>}]
 spine_02 [{shapeType=Capsule, length=2.2205, radius=12.005, rotation=<rot 0 -90 0>}]
 spine_03 [{shapeType=Capsule, length=4.1914, radius=11.5382, rotation=<rot 0 -90 0>}]
 spine_04 [{shapeType=Capsule, length=7.9513, radius=9.718, rotation=<rot 0 -90 0>}]
 spine_05 [{shapeType=Capsule, length=14.3663, radius=10.2969, rotation=<rot 0 -90 0>}]
}
constraints [
 {bone1Name=dyn_cape02, bone2Name=dyn_cape01, swing1LimitDegrees=20, swing2LimitDegrees=20, twistLimitDegrees=5}
 {bone1Name=dyn_cape03, bone2Name=dyn_cape02, swing1LimitDegrees=25, swing2LimitDegrees=20, twistLimitDegrees=5}
 {bone1Name=dyn_cape04, bone2Name=dyn_cape03, swing1LimitDegrees=35, swing2LimitDegrees=20, twistLimitDegrees=5}
 {
  bone1Name=spine_01
  bone2Name=pelvis
  swing1LimitDegrees=45
  swing1Motion=Free
  swing2LimitDegrees=45
  swing2Motion=Free
  twistLimitDegrees=45
  twistMotion=Free
 }
 {
  bone1Name=spine_02
  bone2Name=spine_01
  swing1LimitDegrees=45
  swing1Motion=Free
  swing2LimitDegrees=45
  swing2Motion=Free
  twistLimitDegrees=45
  twistMotion=Free
 }
 {
  bone1Name=spine_03
  bone2Name=spine_02
  swing1LimitDegrees=45
  swing1Motion=Free
  swing2LimitDegrees=45
  swing2Motion=Free
  twistLimitDegrees=45
  twistMotion=Free
 }
 {
  bone1Name=spine_04
  bone2Name=spine_03
  swing1LimitDegrees=45
  swing1Motion=Free
  swing2LimitDegrees=45
  swing2Motion=Free
  twistLimitDegrees=45
  twistMotion=Free
 }
 {
  bone1Name=spine_05
  bone2Name=spine_04
  swing1LimitDegrees=45
  swing1Motion=Free
  swing2LimitDegrees=45
  swing2Motion=Free
  twistLimitDegrees=45
  twistMotion=Free
 }
 {bone1Name=dyn_cape01, bone2Name=dyn_cape05, swing1LimitDegrees=20, swing2LimitDegrees=20, twistLimitDegrees=5}
 {bone1Name=dyn_cape05, bone2Name=backpack_root, swing1LimitDegrees=20, swing2LimitDegrees=20, twistLimitDegrees=5}
]
