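# Animation Blueprint for the Cube Ninja cape cosmetic, simulating the cape with RigidBody physics and setting cloth blend weight by skydive and local-pawn state.
# Each update copies world transforms from the parent mesh's sockets onto body bones, caching the pose.
# Male and female physics-asset simulations read that pose, blended by gender, with gravity and velocity/acceleration scales driven by skydiving or emote state.
# The jump/skydive cape-bone rotation chain is unconnected to the output.
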AnimBlueprint /BRCosmetics/Accessories/FORT_Capes/Cape_M_MED_Cube_Ninja/Meshes/M_MED_Cube_Ninja_Pack_AnimBP
target_skeleton=&/M_MED_Cube_Ninja_Pack_Skeleton
graphs [
 AnimGraph {nodes [
   n1 = "Transform (Modify) Bone - Bone: dyn_cape01_masterRotate"(Alpha=1, Translation="0,0,0", Rotation=Get SkydiveAlpha() -> (SkydiveAlpha:Rotator), Scale="1,1,1") {rotationMode=BMM_Additive, rotationSpace=BCS_BoneSpace, alphaBoolBlend {blendOption=Linear}, boneToModify {boneName=dyn_cape01_masterRotate}}
   n2 = Get JumpAlpha() -> (JumpAlpha:Rotator)
   n3 = "Transform (Modify) Bone - Bone: dyn_cape02"(ComponentPose=@n1, Alpha=1, Translation="0,0,0", Rotation=@n2, Scale="1,1,1") {
    rotationMode=BMM_Additive
    rotationSpace=BCS_BoneSpace
    alphaBoolBlend {blendOption=Linear}
    alphaScaleBiasClamp {bInterpResult=true}
    boneToModify {boneName=dyn_cape02}
   }
   "Transform (Modify) Bone - Bone: dyn_cape03"(ComponentPose=@n3, Alpha=1, Translation="0,0,0", Rotation=@n2, Scale="1,1,1") {
    rotationMode=BMM_Additive
    rotationSpace=BCS_BoneSpace
    alphaBoolBlend {blendOption=Linear}
    alphaScaleBiasClamp {bInterpResult=true}
    boneToModify {boneName=dyn_cape03}
   }
   RigidBody!(
    ComponentPose
    Alpha
    OverrideWorldGravity:Vector="0,0,0"
    ExternalForce
    ComponentLinearAccScale:Vector="0,0,0"
    ComponentLinearVelScale:Vector="0,0,0"
   ) -> (Pose) {
    bClampLinearTranslationLimitToRefPose=true
    bOverrideWorldGravity=true
    overridePhysicsAsset=&/M_MED_Cube_Ninja_Pack_Physics
    simulationSpace=BaseBoneSpace
    alphaBoolBlend {blendOption=Linear}
    baseBoneRef {boneName=body_root}
    componentAppliedLinearAccClamp=<5000 5000 5000>
    componentLinearAccScale=<0.1 0.1 0.1>
    componentLinearVelScale=<0.1 0.1 0.1>
    overrideWorldGravity {z=-980}
    simSpaceSettings {
     maxAngularAcceleration=100
     maxAngularVelocity=100
     maxLinearAcceleration=100
     maxLinearVelocity=100
     velocityScaleZ=0.5
     externalLinearDragV=<10 10 10>
    }
   }
   n4 = Get finalGravity() -> (finalGravity:Vector)
   n5 = Get compLinearAccScale() -> (compLinearAccScale:Vector)
   n6 = Get compLinearVelScale() -> (compLinearVelScale:Vector)
   n7 = RigidBody(
    ComponentPose=Local To Component(LocalPose=Use cached pose 'anim cache'())
    OverrideWorldGravity=@n4
    ComponentLinearAccScale=@n5
    ComponentLinearVelScale=@n6
   ) {
    bClampLinearTranslationLimitToRefPose=true
    bOverrideWorldGravity=true
    overridePhysicsAsset=&M_MED_Cube_Ninja_Pack_Physics
    simulationSpace=BaseBoneSpace
    alphaBoolBlend {blendOption=Linear}
    baseBoneRef {boneName=backpack_root}
    componentAppliedLinearAccClamp=<3000 3000 3000>
    componentLinearAccScale=<0.5 0.4 0.45>
    componentLinearVelScale=<0.2 0.2 0.3>
    overrideWorldGravity {z=-980}
   }
   n8 = RigidBody(
    ComponentPose=Local To Component(LocalPose=Use cached pose 'anim cache'())
    OverrideWorldGravity=@n4
    ComponentLinearAccScale=@n5
    ComponentLinearVelScale=@n6
   ) {
    bClampLinearTranslationLimitToRefPose=true
    bOverrideWorldGravity=true
    overridePhysicsAsset=&/M_MED_Cube_Ninja_Pack_Physics_Female
    simulationSpace=BaseBoneSpace
    alphaBoolBlend {blendOption=Linear}
    baseBoneRef {boneName=backpack_root}
    componentAppliedLinearAccClamp=<3000 3000 3000>
    componentLinearAccScale=<0.5 0.4 0.45>
    componentLinearVelScale=<0.2 0.2 0.3>
    overrideWorldGravity {z=-980}
   }
   Output Pose(Result=Blend Poses by bool(
     BlendPose_0=Component To Local(ComponentPose=@n8)
     BlendPose_1=Component To Local(ComponentPose=@n7)
     bActiveValue=Get IsFemale() -> (IsFemale:bool)
    ) {blendType=Linear}) {layerGroup=None}
   n9 = "Transform (Modify) Bone - Bone: backpack_root"(ComponentPose=Local To Component(LocalPose=Local Space Ref Pose()), Alpha=1, Translation="0,1,0", Rotation="0,0,0", Scale="1,1,1") {translationMode=BMM_Additive, alphaBoolBlend {blendOption=Linear}, boneToModify {boneName=backpack_root}, translation {y=1}}
   n10 = "Transform (Modify) Bone - Bone: body_root"(
    ComponentPose=@n9
    Alpha=Get isEmotingAlpha() -> (isEmotingAlpha:double)
    Translation=Get bodyRoot_tran() -> (bodyRoot_tran:Vector)
    Rotation=Get bodyRoot_rot() -> (bodyRoot_rot:Rotator)
    Scale="1,1,1"
   ) {
    rotationMode=BMM_Replace
    rotationSpace=BCS_WorldSpace
    scaleSpace=BCS_BoneSpace
    translationMode=BMM_Replace
    translationSpace=BCS_WorldSpace
    alphaBoolBlend {blendOption=Linear}
    alphaScaleBiasClamp {inRange {max=0, min=-40}, outRange {max=0, min=1}}
    boneToModify {boneName=body_root}
   }
   n11 = "Transform (Modify) Bone - Bone: pelvis"(
    ComponentPose=@n10
    Alpha=1
    Translation=Get pelvis_tran() -> (pelvis_tran:Vector)
    Rotation=Get pelvis_rot() -> (pelvis_rot:Rotator)
    Scale="1,1,1"
   ) {
    rotationMode=BMM_Replace
    rotationSpace=BCS_WorldSpace
    scaleSpace=BCS_BoneSpace
    translationMode=BMM_Replace
    translationSpace=BCS_WorldSpace
    alphaBoolBlend {blendOption=Linear}
    alphaScaleBiasClamp {inRange {max=0, min=-40}, outRange {max=0, min=1}}
    boneToModify {boneName=pelvis}
   }
   n12 = "Transform (Modify) Bone - Bone: floorCOL"(
    ComponentPose=@n11
    Alpha=1
    Translation=Get root_trans() -> (root_trans:Vector)
    Rotation=Get root_rot() -> (root_rot:Rotator)
    Scale="1,1,1"
   ) {
    rotationMode=BMM_Replace
    rotationSpace=BCS_WorldSpace
    scaleSpace=BCS_BoneSpace
    translationMode=BMM_Replace
    translationSpace=BCS_WorldSpace
    alphaBoolBlend {blendOption=Linear}
    alphaScaleBiasClamp {inRange {max=0, min=-40}, outRange {max=0, min=1}}
    boneToModify {boneName=floorCOL}
   }
   n13 = "Transform (Modify) Bone - Bone: thigh_l"(
    ComponentPose=@n12
    Alpha=1
    Translation=Get thigh_l_tran() -> (thigh_l_tran:Vector)
    Rotation=Get thigh_l_rot() -> (thigh_l_rot:Rotator)
    Scale="1,1,1"
   ) {
    rotationMode=BMM_Replace
    rotationSpace=BCS_WorldSpace
    scaleSpace=BCS_BoneSpace
    translationMode=BMM_Replace
    translationSpace=BCS_WorldSpace
    alphaBoolBlend {blendOption=Linear}
    alphaScaleBiasClamp {inRange {max=0, min=-40}, outRange {max=0, min=1}}
    boneToModify {boneName=thigh_l}
   }
   n14 = "Transform (Modify) Bone - Bone: thigh_r"(
    ComponentPose=@n13
    Alpha=1
    Translation=Get thigh_r_tran() -> (thigh_r_tran:Vector)
    Rotation=Get thigh_r_rot() -> (thigh_r_rot:Rotator)
    Scale="1,1,1"
   ) {
    rotationMode=BMM_Replace
    rotationSpace=BCS_WorldSpace
    scaleSpace=BCS_BoneSpace
    translationMode=BMM_Replace
    translationSpace=BCS_WorldSpace
    alphaBoolBlend {blendOption=Linear}
    alphaScaleBiasClamp {inRange {max=0, min=-40}, outRange {max=0, min=1}}
    boneToModify {boneName=thigh_r}
   }
   n15 = "Transform (Modify) Bone - Bone: calf_l"(
    ComponentPose=@n14
    Alpha=1
    Translation=Get calf_l_tran() -> (calf_l_tran:Vector)
    Rotation=Get calf_l_rot() -> (calf_l_rot:Rotator)
    Scale="1,1,1"
   ) {
    rotationMode=BMM_Replace
    rotationSpace=BCS_WorldSpace
    scaleSpace=BCS_BoneSpace
    translationMode=BMM_Replace
    translationSpace=BCS_WorldSpace
    alphaBoolBlend {blendOption=Linear}
    alphaScaleBiasClamp {inRange {max=0, min=-40}, outRange {max=0, min=1}}
    boneToModify {boneName=calf_l}
   }
   n16 = "Transform (Modify) Bone - Bone: calf_r"(
    ComponentPose=@n15
    Alpha=1
    Translation=Get calf_r_tran() -> (calf_r_tran:Vector)
    Rotation=Get calf_r_rot() -> (calf_r_rot:Rotator)
    Scale="1,1,1"
   ) {
    rotationMode=BMM_Replace
    rotationSpace=BCS_WorldSpace
    scaleSpace=BCS_BoneSpace
    translationMode=BMM_Replace
    translationSpace=BCS_WorldSpace
    alphaBoolBlend {blendOption=Linear}
    alphaScaleBiasClamp {inRange {max=0, min=-40}, outRange {max=0, min=1}}
    boneToModify {boneName=calf_r}
   }
   n17 = "Transform (Modify) Bone - Bone: spine_01"(
    ComponentPose=@n16
    Alpha=1
    Translation=Get spine_01_tran() -> (spine_01_tran:Vector)
    Rotation=Get spine_01_rot() -> (spine_01_rot:Rotator)
    Scale="1,1,1"
   ) {
    rotationMode=BMM_Replace
    rotationSpace=BCS_WorldSpace
    scaleSpace=BCS_BoneSpace
    translationMode=BMM_Replace
    translationSpace=BCS_WorldSpace
    alphaBoolBlend {blendOption=Linear}
    alphaScaleBiasClamp {inRange {max=0, min=-40}, outRange {max=0, min=1}}
    boneToModify {boneName=spine_01}
   }
   n18 = "Transform (Modify) Bone - Bone: spine_02"(
    ComponentPose=@n17
    Alpha=1
    Translation=Get spine_02_tran() -> (spine_02_tran:Vector)
    Rotation=Get spine_02_rot() -> (spine_02_rot:Rotator)
    Scale="1,1,1"
   ) {
    rotationMode=BMM_Replace
    rotationSpace=BCS_WorldSpace
    scaleSpace=BCS_BoneSpace
    translationMode=BMM_Replace
    translationSpace=BCS_WorldSpace
    alphaBoolBlend {blendOption=Linear}
    alphaScaleBiasClamp {inRange {max=0, min=-40}, outRange {max=0, min=1}}
    boneToModify {boneName=spine_02}
   }
   n19 = "Transform (Modify) Bone - Bone: spine_03"(
    ComponentPose=@n18
    Alpha=1
    Translation=Get spine_03_tran() -> (spine_03_tran:Vector)
    Rotation=Get spine_03_rot() -> (spine_03_rot:Rotator)
    Scale="1,1,1"
   ) {
    rotationMode=BMM_Replace
    rotationSpace=BCS_WorldSpace
    scaleSpace=BCS_BoneSpace
    translationMode=BMM_Replace
    translationSpace=BCS_WorldSpace
    alphaBoolBlend {blendOption=Linear}
    alphaScaleBiasClamp {inRange {max=0, min=-40}, outRange {max=0, min=1}}
    boneToModify {boneName=spine_03}
   }
   n20 = "Transform (Modify) Bone - Bone: spine_04"(
    ComponentPose=@n19
    Alpha=1
    Translation=Get spine_04_tran() -> (spine_04_tran:Vector)
    Rotation=Get spine_04_rot() -> (spine_04_rot:Rotator)
    Scale="1,1,1"
   ) {
    rotationMode=BMM_Replace
    rotationSpace=BCS_WorldSpace
    scaleSpace=BCS_BoneSpace
    translationMode=BMM_Replace
    translationSpace=BCS_WorldSpace
    alphaBoolBlend {blendOption=Linear}
    alphaScaleBiasClamp {inRange {max=0, min=-40}, outRange {max=0, min=1}}
    boneToModify {boneName=spine_04}
   }
   n21 = "Transform (Modify) Bone - Bone: spine_05"(
    ComponentPose=@n20
    Alpha=1
    Translation=Get spine_05_tran() -> (spine_05_tran:Vector)
    Rotation=Get spine_05_rot() -> (spine_05_rot:Rotator)
    Scale="1,1,1"
   ) {
    rotationMode=BMM_Replace
    rotationSpace=BCS_WorldSpace
    scaleSpace=BCS_BoneSpace
    translationMode=BMM_Replace
    translationSpace=BCS_WorldSpace
    alphaBoolBlend {blendOption=Linear}
    alphaScaleBiasClamp {inRange {max=0, min=-40}, outRange {max=0, min=1}}
    boneToModify {boneName=spine_05}
   }
   n22 = "Transform (Modify) Bone - Bone: clavicle_r"(
    ComponentPose=@n21
    Alpha=1
    Translation=Get clavicle_r_tran() -> (clavicle_r_tran:Vector)
    Rotation=Get clavicel_r_rot() -> (clavicel_r_rot:Rotator)
    Scale="1,1,1"
   ) {
    rotationMode=BMM_Replace
    rotationSpace=BCS_WorldSpace
    scaleSpace=BCS_BoneSpace
    translationMode=BMM_Replace
    translationSpace=BCS_WorldSpace
    alphaBoolBlend {blendOption=Linear}
    alphaScaleBiasClamp {inRange {max=0, min=-40}, outRange {max=0, min=1}}
    boneToModify {boneName=clavicle_r}
   }
   n23 = "Transform (Modify) Bone - Bone: clavicle_l"(
    ComponentPose=@n22
    Alpha=1
    Translation=Get clavicle_l_tran() -> (clavicle_l_tran:Vector)
    Rotation=Get clavicle_l_rot() -> (clavicle_l_rot:Rotator)
    Scale="1,1,1"
   ) {
    rotationMode=BMM_Replace
    rotationSpace=BCS_WorldSpace
    scaleSpace=BCS_BoneSpace
    translationMode=BMM_Replace
    translationSpace=BCS_WorldSpace
    alphaBoolBlend {blendOption=Linear}
    alphaScaleBiasClamp {inRange {max=0, min=-40}, outRange {max=0, min=1}}
    boneToModify {boneName=clavicle_l}
   }
   n24 = "Transform (Modify) Bone - Bone: upperarm_r"(
    ComponentPose=@n23
    Alpha=1
    Translation=Get upperarm_r_tran() -> (upperarm_r_tran:Vector)
    Rotation=Get upperarm_r_rot() -> (upperarm_r_rot:Rotator)
    Scale="1,1,1"
   ) {
    rotationMode=BMM_Replace
    rotationSpace=BCS_WorldSpace
    scaleSpace=BCS_BoneSpace
    translationMode=BMM_Replace
    translationSpace=BCS_WorldSpace
    alphaBoolBlend {blendOption=Linear}
    alphaScaleBiasClamp {inRange {max=0, min=-40}, outRange {max=0, min=1}}
    boneToModify {boneName=upperarm_r}
   }
   n25 = "Transform (Modify) Bone - Bone: upperarm_l"(
    ComponentPose=@n24
    Alpha=1
    Translation=Get upperarm_l_tran() -> (upperarm_l_tran:Vector)
    Rotation=Get upperarm_l_rot() -> (upperarm_l_rot:Rotator)
    Scale="1,1,1"
   ) {
    rotationMode=BMM_Replace
    rotationSpace=BCS_WorldSpace
    scaleSpace=BCS_BoneSpace
    translationMode=BMM_Replace
    translationSpace=BCS_WorldSpace
    alphaBoolBlend {blendOption=Linear}
    alphaScaleBiasClamp {inRange {max=0, min=-40}, outRange {max=0, min=1}}
    boneToModify {boneName=upperarm_l}
   }
   n26 = "Transform (Modify) Bone - Bone: lowerarm_l"(
    ComponentPose=@n25
    Alpha=1
    Translation=Get lowerarm_l_tran() -> (lowerarm_l_tran:Vector)
    Rotation=Get lowerarm_l_rot() -> (lowerarm_l_rot:Rotator)
    Scale="1,1,1"
   ) {
    rotationMode=BMM_Replace
    rotationSpace=BCS_WorldSpace
    scaleSpace=BCS_BoneSpace
    translationMode=BMM_Replace
    translationSpace=BCS_WorldSpace
    alphaBoolBlend {blendOption=Linear}
    alphaScaleBiasClamp {inRange {max=0, min=-40}, outRange {max=0, min=1}}
    boneToModify {boneName=lowerarm_l}
   }
   n27 = "Transform (Modify) Bone - Bone: lowerarm_r"(
    ComponentPose=@n26
    Alpha=1
    Translation=Get lowerarm_r_tran() -> (lowerarm_r_tran:Vector)
    Rotation=Get lowerarm_r_rot() -> (lowerarm_r_rot:Rotator)
    Scale="1,1,1"
   ) {
    rotationMode=BMM_Replace
    rotationSpace=BCS_WorldSpace
    scaleSpace=BCS_BoneSpace
    translationMode=BMM_Replace
    translationSpace=BCS_WorldSpace
    alphaBoolBlend {blendOption=Linear}
    alphaScaleBiasClamp {inRange {max=0, min=-40}, outRange {max=0, min=1}}
    boneToModify {boneName=lowerarm_r}
   }
   n28 = "Constraint - root (RRR)"!(ComponentPose=@n27, Alpha=1, ConstraintWeights_0=0.4, ConstraintWeights_1:float=0.5, ConstraintWeights_2:float=0.1) -> (Pose) {
    alphaBoolBlend {blendOption=Linear}
    boneToModify {boneName=root}
    constraintSetup [
     {transformType=Rotation, targetBone {boneName=spine_05}}
     {transformType=Rotation, targetBone {boneName=spine_04}}
     {transformType=Rotation, targetBone {boneName=spine_03}}
    ]
    constraintWeights [1,1,1]
   }
   Save cached pose 'anim cache'(Pose=Component To Local(ComponentPose=@n28)) {cacheName=anim cache}
   Get rbanOnOff() -> (rbanOnOff:double)
   Get GravityOverride()
  ]}
 EventGraph {nodes [
   Event BlueprintUpdateAnimation()
   > n1 = Sequence!(execute) -> (then_0, then_1, then_2:Exec, then_3:Exec, then_4:Exec, then_5:Exec, then_6:Exec)
   n2 = Break Fort Anim Input Skydiving External Force(FortAnimInput_SkydivingExternalForce=Get SkydivingExternalForceInput()) -> (
    bUseSkydivingVectorForce=~
    SkydivingVectorForce=~
    HeadToPelvisDirection=~
    FloatingMultiplier=~
    FloatingAdditive=~
    DivingMultiplier=~
    DivingAdditive=~
    ParachutingMultiplier=~
    ParachutingAdditive=~
   )
   > n3 = Set windVector(windVector:Vector=SelectVector(A=vector * float(A=@n2.SkydivingVectorForce, B=0.1), B="0, 0,-735", bPickA=Get bIsSkyDiving())) -> (Output_Get:Vector)
   n4 = GetOwnerSkeletalMeshComponent()
   n5 = BreakRotator(InRot=BreakTransform(InTransform=GetDeltaTransformFromRefPose(self=@n4, BoneName=pelvis, BaseName=floorCOL)).Rotation)
   n6 = SelectFloat(A=980, B=-980, bPickA="float > float"(A="Absolute (Float)"(A=@n5.Yaw), B=90) {is=title K2Node_CallFunction})
   > Set finalGravity(finalGravity:Vector=SelectVector(A=MakeVector(Y=0, Z=@n6), B=@n3, bPickA=Get bIsPlayingEmote())) -> (Output_Get:Vector)
   n7 = Branch(execute=@n1.then_1, Condition=Get bIsSkyDiving())
   n8 = GetPartSkeletalMeshComponent()
   > Set ClothBlendWeight!(execute, ClothBlendWeight:float=0.5, self:Skeletal Mesh Component Object Reference=@n8) -> (then, Output_Get:float)
   n9 = Branch(execute=@n7.else, Condition=IsLocalPawn!(Actor:Actor Object Reference=GetOwningActor()) -> (IsLocalPawn:bool))
   > Set ClothBlendWeight(ClothBlendWeight=1, self=@n8)
   Set ClothBlendWeight(execute=@n9.else, ClothBlendWeight=0, self=@n8)
   n10 = Get bRigidBodyEnabled()
   n11 = Get CurrentLOD()
   n12 = Set rigidsOnOffBool(
    execute=@n1.then_3
    rigidsOnOffBool:bool=OR Boolean!(
     A=NOT Boolean(A=@n10)
     B="integer > integer"(A=@n11, B=2) {is=title K2Node_CallFunction}
     C:bool=Get bCanPlayCustomAnimations() -> (bCanPlayCustomAnimations:bool)
     D:bool=Get bIsPlayingEmote()
     E:bool
    ) -> (ReturnValue)
   ) -> (Output_Get:bool)
   > Set rbanOnOff(rbanOnOff:double="To Float (Boolean)"(InBool=NOT Boolean(A=@n12))) -> (Output_Get:double)
   n13 = SelectVector(A="0.1,0.1,0.1", B="0.3,0.3,0.45", bPickA=Get bIsSkyDiving())
   Set compLinearVelScale(execute=@n1.then_4, compLinearVelScale:Vector=@n13) -> (Output_Get:Vector)
   > Set compLinearAccScale(compLinearAccScale:Vector=@n13) -> (Output_Get:Vector)
   n14 = Cast To FortPlayerPawn(execute=@n1.then_5, Object=TryGetPawnOwner())
   n15 = GetGender!(self:Fort Player Pawn Object Reference=@n14) -> (IsMale:bool, IsFemale:bool)
   > Set IsMale(IsMale:bool=@n15) -> (Output_Get:bool)
   > Set IsFemale(IsFemale:bool=@n15.IsFemale) -> (Output_Get:bool)
   Set isEmotingAlpha(execute=@n1.then_6, isEmotingAlpha:double=SelectFloat(A=1, B=0.5, bPickA=Get bIsPlayingEmote())) -> (Output_Get:double)
   n16 = GetPartSkeletalMeshComponent()
   Is Valid(exec=@n1.then_2, InputObject=@n16)
   n17 = GetAttachParent!(self:Scene Component Object Reference=@n16) -> (ReturnValue:Scene Component Object Reference)
   > Is Valid(InputObject=@n17)
   > n18 = Sequence!(execute) -> (
    then_0
    then_1
    then_2
    then_3
    then_4
    then_5
    then_6
    then_7:Exec
    then_8:Exec
    then_9:Exec
    then_10:Exec
    then_11:Exec
    then_12:Exec
    then_13:Exec
    then_14:Exec
    then_15:Exec
    then_16:Exec
    then_17:Exec
   )
   n19 = BreakTransform(InTransform=GetSocketTransform(self=@n17, InSocketName=spine_05, TransformSpace=RTS_World))
   Set spine_05_tran(execute=@n18.then_10, spine_05_tran:Vector=@n19) -> (Output_Get:Vector)
   > Set spine_05_rot(spine_05_rot:Rotator=@n19.Rotation) -> (Output_Get:Rotator)
   n20 = BreakTransform(InTransform=GetSocketTransform(self=@n17, InSocketName=clavicle_r, TransformSpace=RTS_World))
   Set clavicle_r_tran(execute=@n18.then_11, clavicle_r_tran:Vector=@n20) -> (Output_Get:Vector)
   > Set clavicel_r_rot(clavicel_r_rot:Rotator=@n20.Rotation) -> (Output_Get:Rotator)
   n21 = BreakTransform(InTransform=GetSocketTransform(self=@n17, InSocketName=clavicle_l, TransformSpace=RTS_World))
   Set clavicle_l_tran(execute=@n18.then_12, clavicle_l_tran:Vector=@n21) -> (Output_Get:Vector)
   > Set clavicle_l_rot(clavicle_l_rot:Rotator=@n21.Rotation) -> (Output_Get:Rotator)
   n22 = BreakTransform(InTransform=GetSocketTransform(self=@n17, InSocketName=upperarm_r, TransformSpace=RTS_World))
   Set upperarm_r_tran(execute=@n18.then_13, upperarm_r_tran:Vector=@n22) -> (Output_Get:Vector)
   > Set upperarm_r_rot(upperarm_r_rot:Rotator=@n22.Rotation) -> (Output_Get:Rotator)
   n23 = BreakTransform(InTransform=GetSocketTransform(self=@n17, InSocketName=upperarm_l, TransformSpace=RTS_World))
   Set upperarm_l_tran(execute=@n18.then_14, upperarm_l_tran:Vector=@n23) -> (Output_Get:Vector)
   > Set upperarm_l_rot(upperarm_l_rot:Rotator=@n23.Rotation) -> (Output_Get:Rotator)
   n24 = BreakTransform(InTransform=GetSocketTransform(self=@n17, InSocketName=lowerarm_l, TransformSpace=RTS_World))
   Set lowerarm_l_tran(execute=@n18.then_15, lowerarm_l_tran:Vector=@n24) -> (Output_Get:Vector)
   > Set lowerarm_l_rot(lowerarm_l_rot:Rotator=@n24.Rotation) -> (Output_Get:Rotator)
   n25 = BreakTransform(InTransform=GetSocketTransform(self=@n17, InSocketName=lowerarm_r, TransformSpace=RTS_World))
   Set lowerarm_r_tran(execute=@n18.then_16, lowerarm_r_tran:Vector=@n25) -> (Output_Get:Vector)
   > Set lowerarm_r_rot(lowerarm_r_rot:Rotator=@n25.Rotation) -> (Output_Get:Rotator)
   n26 = BreakTransform(InTransform=GetSocketTransform(self=@n17, InSocketName=root, TransformSpace=RTS_World))
   Set root_trans(execute=@n18.then_17, root_trans:Vector=@n26) -> (Output_Get:Vector)
   > Set root_rot(root_rot:Rotator=@n26.Rotation) -> (Output_Get:Rotator)
   n27 = BreakTransform(InTransform=GetSocketTransform(self=@n17, InSocketName=spine_01, TransformSpace=RTS_World))
   Set spine_01_tran(execute=@n18.then_6, spine_01_tran:Vector=@n27) -> (Output_Get:Vector)
   > Set spine_01_rot(spine_01_rot:Rotator=@n27.Rotation) -> (Output_Get:Rotator)
   n28 = BreakTransform(InTransform=GetSocketTransform(self=@n17, InSocketName=spine_02, TransformSpace=RTS_World))
   Set spine_02_tran(execute=@n18.then_7, spine_02_tran:Vector=@n28) -> (Output_Get:Vector)
   > Set spine_02_rot(spine_02_rot:Rotator=@n28.Rotation) -> (Output_Get:Rotator)
   n29 = BreakTransform(InTransform=GetSocketTransform(self=@n17, InSocketName=spine_03, TransformSpace=RTS_World))
   Set spine_03_tran(execute=@n18.then_8, spine_03_tran:Vector=@n29) -> (Output_Get:Vector)
   > Set spine_03_rot(spine_03_rot:Rotator=@n29.Rotation) -> (Output_Get:Rotator)
   n30 = BreakTransform(InTransform=GetSocketTransform(self=@n17, InSocketName=spine_04, TransformSpace=RTS_World))
   Set spine_04_tran(execute=@n18.then_9, spine_04_tran:Vector=@n30) -> (Output_Get:Vector)
   > Set spine_04_rot(spine_04_rot:Rotator=@n30.Rotation) -> (Output_Get:Rotator)
   n31 = BreakTransform(InTransform=GetSocketTransform(self=@n17, InSocketName=pelvis, TransformSpace=RTS_World))
   Set pelvis_tran(execute=@n18.then_0, pelvis_tran:Vector=@n31) -> (Output_Get:Vector)
   > Set pelvis_rot(pelvis_rot:Rotator=@n31.Rotation) -> (Output_Get:Rotator)
   n32 = BreakTransform(InTransform=GetSocketTransform(self=@n17, InSocketName=thigh_l, TransformSpace=RTS_World))
   Set thigh_l_tran(execute=@n18.then_1, thigh_l_tran:Vector=@n32) -> (Output_Get:Vector)
   > Set thigh_l_rot(thigh_l_rot:Rotator=@n32.Rotation) -> (Output_Get:Rotator)
   n33 = BreakTransform(InTransform=GetSocketTransform(self=@n17, InSocketName=thigh_r, TransformSpace=RTS_World))
   Set thigh_r_tran(execute=@n18.then_2, thigh_r_tran:Vector=@n33) -> (Output_Get:Vector)
   > Set thigh_r_rot(thigh_r_rot:Rotator=@n33.Rotation) -> (Output_Get:Rotator)
   n34 = BreakTransform(InTransform=GetSocketTransform(self=@n17, InSocketName=calf_r, TransformSpace=RTS_World))
   Set calf_r_tran(execute=@n18.then_4, calf_r_tran:Vector=@n34) -> (Output_Get:Vector)
   > Set calf_r_rot(calf_r_rot:Rotator=@n34.Rotation) -> (Output_Get:Rotator)
   n35 = BreakTransform(InTransform=GetSocketTransform(self=@n17, InSocketName=calf_l, TransformSpace=RTS_World))
   Set calf_l_tran(execute=@n18.then_3, calf_l_tran:Vector=@n35) -> (Output_Get:Vector)
   > Set calf_l_rot(calf_l_rot:Rotator=@n35.Rotation) -> (Output_Get:Rotator)
   n36 = BreakTransform(InTransform=GetSocketTransform(self=@n17, InSocketName=root, TransformSpace=RTS_World))
   Set bodyRoot_tran(execute=@n18.then_5, bodyRoot_tran:Vector=@n36) -> (Output_Get:Vector)
   > Set bodyRoot_rot(bodyRoot_rot:Rotator=@n36.Rotation) -> (Output_Get:Rotator)
   Event BlueprintInitializeAnimation()
   > Set isEmotingAlpha(isEmotingAlpha=0.5)
   Sequence!(execute) -> (then_0, then_1, then_2)
   > n37 = Branch(Condition=Get bIsSwimSprinting() -> (bIsSwimSprinting:bool))
   > Set GravityOverride(GravityOverride="0, 0, 0")
   n38 = Branch(execute=@n37.else, Condition=Get bIsSkyDiving())
   > n39 = Branch(Condition=Get bIsParachuteOpened())
   Set GravityOverride(execute=[@n39.then, @n38.else], GravityOverride="0, 0,-980")
   n40 = Branch(execute=@n39.else, Condition="float > float"(A=Get PawnVelocityZ(), B=40))
   > Set GravityOverride(GravityOverride="0, 0,-980")
   Set GravityOverride(execute=@n40.else, GravityOverride="0, 0,-980")
   TryGetPawnOwner()
   TryGetPawnOwner()
   TryGetPawnOwner()
   TryGetPawnOwner()
   Event OnBeginPlayingEmote() {is=title K2Node_Event}
   Event OnEndPlayingEmote() {is=title K2Node_Event}
   K2Node_Event_4 = CustomEvent!() -> (OutputDelegate, then, DeltaTimeX:double=~) {is=title K2Node_CustomEvent}
   Set rbanOnOff(rbanOnOff=1)
   n41 = Get PawnSpeed2D()
   Set isCrouchingAlpha(isCrouchingAlpha:double=SelectFloat(
     A=MapRangeClamped(Value=@n41, InRangeA=0, InRangeB=419, OutRangeA=1, OutRangeB=0.5)
     B=MapRangeClamped(Value=@n41, InRangeA=400, InRangeB=550, OutRangeA=1, OutRangeB=0.75)
     bPickA=Get bIsCrouching()
    )) -> (Output_Get:double)
   Set rbanOnOff(rbanOnOff=1)
   Set rbanOnOff(rbanOnOff=0)
   Set JumpAlpha(JumpAlpha:Rotator=SelectRotator!(A:Rotator="0,-10,0", B:Rotator="0,5, 0", bPickA:bool=Get bIsFalling()) -> (ReturnValue:Rotator)) -> (Output_Get:Rotator)
   Set SkydiveAlpha(SkydiveAlpha:Rotator=SelectRotator(A="0,-25,0", B="0,20, 0", bPickA=Get bIsSkyDiving())) -> (Output_Get:Rotator)
  ]}
]
variables [
 bodyRoot_rot {type=Rotator}
 bodyRoot_tran {type=Vector}
 calf_l_rot {type=Rotator}
 calf_l_tran {type=Vector}
 calf_r_rot {type=Rotator}
 calf_r_tran {type=Vector}
 clavicel_r_rot {type=Rotator}
 clavicle_l_rot {type=Rotator}
 clavicle_l_tran {type=Vector}
 clavicle_r_tran {type=Vector}
 compLinearAccScale {type=Vector}
 compLinearVelScale {type=Vector}
 finalGravity {type=Vector}
 GravityOverride {type=Vector}
 isCrouchingAlpha {type=number}
 isEmotingAlpha {type=number}
 IsFemale {type=boolean}
 IsMale {type=boolean}
 JumpAlpha {type=Rotator}
 lowerarm_l_rot {type=Rotator}
 lowerarm_l_tran {type=Vector}
 lowerarm_r_rot {type=Rotator}
 lowerarm_r_tran {type=Vector}
 pelvis_rot {type=Rotator}
 pelvis_tran {type=Vector}
 rbanOnOff {type=number}
 rigidsOnOffBool {type=boolean}
 root_rot {type=Rotator}
 root_trans {type=Vector}
 SkydiveAlpha {type=Rotator}
 spine_01_rot {type=Rotator}
 spine_01_tran {type=Vector}
 spine_02_rot {type=Rotator}
 spine_02_tran {type=Vector}
 spine_03_rot {type=Rotator}
 spine_03_tran {type=Vector}
 spine_04_rot {type=Rotator}
 spine_04_tran {type=Vector}
 spine_05_rot {type=Rotator}
 spine_05_tran {type=Vector}
 thigh_l_rot {type=Rotator}
 thigh_l_tran {type=Vector}
 thigh_r_rot {type=Rotator}
 thigh_r_tran {type=Vector}
 upperarm_l_rot {type=Rotator}
 upperarm_l_tran {type=Vector}
 upperarm_r_rot {type=Rotator}
 upperarm_r_tran {type=Vector}
 windVector {type=Vector}
]
